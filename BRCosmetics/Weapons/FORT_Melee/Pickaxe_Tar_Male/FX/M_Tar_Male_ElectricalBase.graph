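# Translucent, unlit, two-sided electrical effect material for Niagara mesh particles on the Tar Male pickaxe.
# Inverted Fresnel terms drive particle-tinted glow color, remapped to HDR except on mobile, and a particle-alpha-scaled mask.
# Panning noise builds a displacement vector, scaled by a dynamic parameter and boosted as the actor gets farther from camera.
# That vector feeds separate multiplies by inverse vertex red and by vertex normal.
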Material /BRCosmetics/Weapons/FORT_Melee/Pickaxe_Tar_Male/FX/M_Tar_Male_ElectricalBase
graph {
 properties {
  bUsedWithNiagaraMeshParticles=true
  blendMode=BLEND_Translucent
  pixelDepthOffsetMode=PDOM_Legacy
  shadingModel=MSM_Unlit
  twoSided=true
  usedShadingModels=Unlit
 }
 nodes [
  n1 = VectorParameter() {group=ElementalElectric, parameterName=GlowColor_Highlights, defaultValue=<rgba 0.6875 0.6875 0.6875 1>}
  n2 = ScalarParameter() {defaultValue=1.25, group=ElementalElectric, parameterName=ElectricFresnelBrightnessColor}
  n3 = ScalarParameter() {defaultValue=3, group=ElementalElectric, parameterName=ElectricFresnelExponentColor}
  n4 = ParticleColor()
  n5 = Multiply(
   A=LinearInterpolate(
    A=@n1.RGB
    B=VectorParameter() {group=ElementalElectric, parameterName=GlowColor_Base, defaultValue=<rgba 4 4 4 0>}.RGB
    Alpha=MaterialFunctionCall(Brightness=@n2, Exponent=@n3) {materialFunction=&MF_Fresnel_FX}.Fresnel Inverted
   ) {constB=0}
   B=@n4.RGB
  )
  n6 = Clamp(None=@n5)
  MaterialFunctionCall(Emissive=ShadingPathSwitch(
    Default=MaterialFunctionCall(LDR Input=@n6, HDR Tint=@n5) {materialFunction=&DeriveHDRfromLDR}.RemappedRange
    Mobile=@n5
   )) {materialFunction=&MF_FX_EmissiveAdjustments}
  n7 = Multiply(
   A=VectorParameter() {group=ElementalElectric, parameterName=WPOScale, defaultValue=<rgba 1.5 1.5 1.25 0>}.RGB
   B=DynamicParameter() {paramNames [WPOScaleFactor, Param2, Param3, Param4]}.WPOScaleFactor
  )
  n8 = Multiply(A=@n7, B=ScalarParameter() {defaultValue=5, parameterName=Distance WPO Factor})
  n9 = OneMinus(None=SphereMask(A=ActorPositionWS(), B=CameraPositionWS(), Radius=ScalarParameter() {defaultValue=1500, parameterName=Distanve}) {hardnessPercent=50})
  n10 = Panner(Coordinate=Multiply(A=TextureCoordinate(), B=ScalarParameter() {defaultValue=0.25, group=ElementalElectric, parameterName=NoiseUVScale})) {bFractionalPart=true, speedY=-1}
  n11 = Add(
   A=Multiply(
    A=LinearInterpolate(A=@n7, B=@n8, Alpha=@n9)
    B=TextureSample(UVs=@n10) {samplerType=SAMPLERTYPE_Normal, texture=/Game/Global/Textures/Default/Noise/64x64Noise}.RGB
   )
   B=VectorParameter() {group=ElementalElectric, parameterName=WPOBias, defaultValue=<rgba 0.1 0.1 0 0>}.RGB
  )
  Multiply(A=@n11, B=OneMinus(None=VertexColor().R))
  Multiply(
   A=@n4.A
   B=MaterialFunctionCall(
    Brightness=ScalarParameter() {defaultValue=2, group=ElementalElectric, parameterName=ElectricFresnelBrightness}
    Exponent=ScalarParameter() {defaultValue=2, group=ElementalElectric, parameterName=ElectricFresnelExponent}
   ) {materialFunction=&MF_Fresnel_FX}.Fresnel Inverted
  )
  Multiply(A=VertexNormalWS(), B=@n11)
 ]
}
