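# Niagara swing effect for the Tar Male pickaxe, drawing a brief orange HDR ribbon stroke that reads as flickering lightning.
# A single emitter spawns ribbon points just above the simulation position over a short 0.2-second window, with lifetimes shrinking across it.
# Brightness scales by time of day and by a stepped flicker curve on emitter age.
# Dynamic material parameters set one fixed value at spawn and animate a second downward along a curve.
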NiagaraSystem /BRCosmetics/Weapons/FORT_Melee/Pickaxe_Tar_Male/FX/NS_Pickaxe_TarMale_Swing
dependencies {usedDataInterfaces [&NiagaraDataInterfaceCurve], usedDynamicInputs [], usedModules [&SystemState]}
emitters {Single_Stroke_Lightinging {
  emitterUpdateScript [
   &EmitterState(Life Cycle Mode=::NewEnumerator1, Loop Behavior=::NewEnumerator1, Loop Duration=0.2, UseLoopDelay=false)
   &SpawnRate(SpawnRate=50, Use Spawn Probability=false)
  ]
  particleSpawnScript [
   &InitializeRibbon(
    Color=<rgba 4 0.36 0.1 1>
    Lifetime=&FloatFromCurve(FloatCurve=DataInterface() {propertyValues {Curve {keys [<at 0 0.2>,<at 1 0.05>]}}}, CurveIndex=$Emitter.NormalizedLoopAge)
    Module.Write Color=true
    Module.Write Lifetime=true
    Module.Write Mass=true
    Module.Write Material Random=true
    Module.Write Position=true
    Module.Write RibbonTwist=false
    Module.Write RibbonWidth=true
    Position:Vector3f=&Add_Vector(A=/Niagara/DynamicInputs/Helpers/SimulationPosition(), B=<0 0 5>)
    Ribbon Width=25
   )
   &DynamicMaterialParameters(
    Index 0 Param 1=0.75
    Module.Param0WriteEnabled=true
    Module.Param1WriteEnabled=false
    Module.Param2WriteEnabled=false
    Module.Param3WriteEnabled=false
   )
  ]
  particleUpdateScript [
   &ParticleState()
   &SolveForcesAndVelocity(Clamp Velocity=false, Limit Acceleration=false)
   &ScaleColor(
    Scale RGB=&VectorFromFloat(Value=&FNTimeOfDayScalar(Night Scale=4, Morning Scale=4, Day Scale=5, Evening Scale=5))
    ScaleA=true
    ScaleRGB=true
   )
   &ScaleColor(
    Scale RGB=&VectorFromFloat(Value=&FloatFromCurve(
      FloatCurve=DataInterface() {propertyValues {Curve {keys [<at 0 1 RCIM_Constant>,<at 0.05 5 RCIM_Constant>,<at 0.15 0.5 RCIM_Constant>,<at 0.22 3 RCIM_Constant>,<at 0.416 1 RCIM_Constant>,<at 0.5493 2 RCIM_Constant>,<at 0.64 0.5 RCIM_Constant>,<at 0.7224 5 RCIM_Constant>,<at 0.799 0.75 RCIM_Constant>,<at 0.8826 3 RCIM_Constant>,<at 1 1>]}}}
      CurveIndex=$Emitter.NormalizedLoopAge
     ))
    ScaleA=true
    ScaleRGB=true
   ) {module=ScaleColor001}
   &DynamicMaterialParameters(
    Index 0 Param 2=&FloatFromCurve(FloatCurve=DataInterface() {propertyValues {Curve {keys [<at 0 0 RCIM_Cubic>, {arriveTangent=-6.7156, interpMode=RCIM_Cubic, leaveTangent=-6.7156, tangentMode=RCTM_User, time=1, value=-2}]}}})
    Module.Param0WriteEnabled=false
    Module.Param1WriteEnabled=true
    Module.Param2WriteEnabled=false
    Module.Param3WriteEnabled=false
   ) {module=DynamicMaterialParameters001}
  ]
  rendererClasses [&NiagaraRibbonRendererProperties]
  renderers [{rendererClass=&NiagaraRibbonRendererProperties}]
 }}
summary {
 systemName=NS_Pickaxe_TarMale_Swing
 emitters [{emitterName=Single_Stroke_Lightinging, rendererClasses [&NiagaraRibbonRendererProperties]}]
 userVariables [
  User.RibbonIDOffset {type=NiagaraInt32}
  User.SkeletalMesh {
   type=NiagaraDataInterfaceSkeletalMesh
   defaultValue {
    struct=NiagaraExt_VariableValue_DataInterface
    value {dataInterface=&self:NiagaraDataInterfaceSkeletalMesh_1}
   }
  }
 ]
}
