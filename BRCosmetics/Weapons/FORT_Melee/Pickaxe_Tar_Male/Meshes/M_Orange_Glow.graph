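# M_Orange_Glow is a Default Lit material in the Pickaxe_Tar_Male weapon mesh folder, flagged for skeletal mesh and mesh deformer use.
# It holds two texture samples that reference the Tar backpack's _D texture and its _S texture, the latter sampled as masks, rather than pickaxe textures.
# No connections to material outputs and no emissive or color logic appear, despite the orange glow name.
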Material /BRCosmetics/Weapons/FORT_Melee/Pickaxe_Tar_Male/Meshes/M_Orange_Glow
graph {
 properties {bUsedWithMeshDeformer=true, bUsedWithSkeletalMesh=true, pixelDepthOffsetMode=PDOM_Legacy, usedShadingModels=Default Lit}
 nodes [
  TextureSample() {texture=/BRCosmetics/Accessories/FORT_Backpacks/Backpack_M_MED_Tar/Textures/T_M_Med_Tar_Backpack_D}
  TextureSample() {samplerType=SAMPLERTYPE_Masks, texture=/BRCosmetics/Accessories/FORT_Backpacks/Backpack_M_MED_Tar/Textures/T_M_Med_Tar_Backpack_S}
 ]
}
